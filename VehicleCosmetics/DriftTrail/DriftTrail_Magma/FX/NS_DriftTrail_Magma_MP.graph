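# Niagara system for the Magma vehicle drift trail cosmetic, drawing molten fire effects at the wheels.
# Five emitters layer SubUV sprites, three spawn groups of fire, twin ribbons, vortex-driven swirls and wheel-wrap meshes scaled per wheel.
# Particles attach to skeletal-mesh sockets by partition or ribbon, shifted by a user offset except the wraps.
# Each emitter picks colors through painted-variant overrides and masks particles per tire with FN_DisableTireFX.
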NiagaraSystem /VehicleCosmetics/DriftTrail/DriftTrail_Magma/FX/NS_DriftTrail_Magma_MP
dependencies {
 usedDataInterfaces [
  &NiagaraDataInterfaceCurve
  &NiagaraDataInterfaceSpriteRendererInfo
  &NiagaraDataInterfaceVectorCurve
  &NiagaraDataInterfaceSkeletalMesh
 ]
 usedDynamicInputs []
 usedModules [
  &self:SystemSpawnScript.SystemScriptSource.SystemScriptGraph.NiagaraNodeAssignment_0.SetVariables_#1
  &SystemState
  &FN_UnpackVehicleFlags
  /Game/Effects/Niagara/Modules/Vehicle/FN_UpdateVehicleAttribs
  /Game/Effects/Niagara/Modules/Vehicle/FN_VehicleFlags_Cheatcodes
 ]
}
emitters {
 NE_VC_DriftTrail_DB001 {
  emitterUpdateScript [
   &EmitterState(Inactive Response=::NewEnumerator1, Life Cycle Mode=::NewEnumerator1, Loop Duration=0.1, UseLoopDelay=false)
   &SpawnRate(SpawnRate=50, Use Spawn Probability=false)
  ]
  particleSpawnScript [
   &VC_PaintedVariantsOverride(
    Black=<rgba 0.01 0.01 0.01 1>
    BurntSienna=<rgba 0.9 0.18 0 1>
    Cobalt=<rgba 0.175 0.3412 3.5 1>
    Crimson=&RandomRangeLinearColor(Minimum=<rgba 0.1 0 0 1>, Maximum=<rgba 1.5 0.045 0.045 1>)
    DefaultColor=&RandomRangeLinearColor(Minimum=<rgba 0.5 0.0083 0 1>, Maximum=<rgba 3 0.4 0 1>)
    ForestGreen=<rgba 0.03 1 0.03 1>
    Grey=<rgba 0.35 0.35 0.35 1>
    Lime=<rgba 0.925 1.5 0 1>
    Orange=<rgba 1.5 0.2 0 1>
    Pink=<rgba 1.8 0.54 1.59 1>
    Purple=&RandomRangeLinearColor(Minimum=<rgba 0.1667 0 0.4 1>, Maximum=<rgba 0.2917 0 0.7 1>)
    Saffron=<rgba 1.4 1.5 0 1>
    SkyBlue=<rgba 0 0.9 1.8 1>
    TitaniumWhite=<rgba 0.8 0.8 0.8 1>
   )
   &InitializeParticle(
    Color=$Particles.Color
    Lifetime Max=0.4
    Lifetime Min=0.3
    Lifetime Mode=::NewEnumerator1
    Sprite Rotation Mode=::NewEnumerator2
    Sprite Size Mode=::NewEnumerator1
    Uniform Sprite Size Max=45
    Uniform Sprite Size Min=40
    UsePositionOffset=false
    Write Material Random=true
   )
   &PartitionParticles()
   &SkeletalMeshLocation(
    Bone / Socket Index=&Add_Integer(A=$Particles.PartitionParticles.Partition, B=4)
    Bone / Socket Sampling Mode=::NewEnumerator5
    Interpolate Mesh Sampling=false
    MaskSpawnGroup=false
    Mesh Sampling Type=::NewEnumerator6
    Sampled Position Offset=&AddVectorToPosition(Position=<-30 0 5>, Vector=$User.offset)
    Skeletal Mesh=$System.SkeletalMesh
   )
   &AddVelocity(
    Cone Angle=10
    Cone Axis=<-1 0 0>
    Inner Cone Angle=5
    Module.UseConeFalloff=false
    Module.UseVelDistribution=false
    Override Seed=false
    Velocity Mode=::NewEnumerator2
    Velocity Speed=&RandomRangeFloat(Minimum=0, Maximum=150)
   )
  ]
  particleUpdateScript [
   &FN_DisableTireFX(
    Mask Index Attribute:NiagaraInt32=$Particles.PartitionParticles.Partition
    Mask Index FL=2
    Mask Index FR=3
    Mask Index RL=0
    Mask Index RR=1
   )
   &ParticleState()
   &ScaleSpriteSize(Uniform Curve Sprite Scale=DataInterface() {propertyValues {Curve {
       defaultValue=0
       keys [
        {arriveTangent=4.1785, interpMode=RCIM_Cubic, leaveTangent=4.1785, tangentMode=RCTM_User, time=0, value=0}
        {arriveTangent=1.1179, interpMode=RCIM_Cubic, leaveTangent=1.1179, tangentMode=RCTM_User, time=0.1, value=0.8}
        {arriveTangent=0.0224, interpMode=RCIM_Cubic, leaveTangent=0.0224, tangentMode=RCTM_User, time=1, value=1}
       ]
      }}})
   &SolveForcesAndVelocity(Clamp Velocity=false, Limit Acceleration=false)
   &ScaleColor(
    Scale Alpha=&FloatFromCurve(FloatCurve=DataInterface() {propertyValues {Curve {
        defaultValue=1
        keys [
         {interpMode=RCIM_Cubic, tangentMode=RCTM_User, time=0, value=0}
         {interpMode=RCIM_Cubic, tangentMode=RCTM_User, time=0.5, value=1}
         {tangentMode=RCTM_User, time=1, value=0}
        ]
       }}})
    Scale RGB=&VectorFromFloat(Value=&FloatFromCurve(
      FloatCurve=DataInterface() {propertyValues {Curve {
         defaultValue=1
         keys [
          {interpMode=RCIM_Cubic, leaveTangentWeight=0.3333, tangentMode=RCTM_User, time=0, value=1}
          {arriveTangentWeight=0.3333, interpMode=RCIM_Cubic, tangentMode=RCTM_User, time=1, value=0.4}
         ]
        }}}
      Scale Curve=1.3
     ))
    ScaleA=true
    ScaleRGB=true
   )
   &SubUVAnimation(
    "Sprite Renderer "=DataInterface() {propertyValues {SpriteRenderer=&self:NiagaraEmitter_0.NiagaraSpriteRendererProperties_3}}
    UseEndFrame=false
    UseStartFrame=false
   )
  ]
  rendererClasses [&NiagaraSpriteRendererProperties]
  renderers [{rendererClass=&NiagaraSpriteRendererProperties}]
 }
 NE_VC_DriftTrail_Fire001 {
  emitterUpdateScript [
   &EmitterState(Inactive Response=::NewEnumerator1, Life Cycle Mode=::NewEnumerator1, Loop Duration=1, UseLoopDelay=false)
   &SpawnRate(SpawnRate=61, Use Spawn Probability=false)
   &SpawnRate(SpawnGroup=1, SpawnRate=41, Use Spawn Probability=false) {module=SpawnRate001}
   &SpawnRate(SpawnGroup=2, SpawnRate=30, Use Spawn Probability=false) {module=SpawnRate002}
  ]
  particleSpawnScript [
   &VC_PaintedVariantsOverride(
    Black=<rgba 0.01 0.01 0.01 1>
    BurntSienna=<rgba 0.3 0.06 0 1>
    Cobalt=<rgba 0.08 0.2 0.8 1>
    Crimson=<rgba 0.5 0 0 1>
    DefaultColor=<rgba 1 0.1667 0 1>
    ForestGreen=<rgba 0 0.6 0 1>
    Grey=<rgba 0.11 0.11 0.11 1>
    Lime=<rgba 0.555 0.9 0 1>
    Orange=<rgba 0.8 0.1067 0 1>
    Pink=<rgba 0.65 0.195 0.5742 1>
    Purple=<rgba 0.1667 0 0.4 1>
    R1 Black:LinearColor=<rgba 0.3 0.3 0.3 1>
    R1 BurntSienna:LinearColor=<rgba 2.5 0.5 0 1>
    R1 Cobalt:LinearColor=<rgba 0.24 0.47 3 1>
    R1 Crimson:LinearColor=<rgba 3 0 0 1>
    R1 DefaultColor=<rgba 3 0.1 0 1>
    R1 ForestGreen:LinearColor=<rgba 0 2.1 0 1>
    R1 Grey:LinearColor=<rgba 2 2 2 1>
    R1 Lime:LinearColor=<rgba 1.85 3 0 1>
    R1 Orange:LinearColor=<rgba 3 0.4 0 1>
    R1 Pink:LinearColor=<rgba 3 0.9 2.65 1>
    R1 Purple:LinearColor=<rgba 1.0417 0 2.5 1>
    R1 Saffron:LinearColor=<rgba 2.8 3 0 1>
    R1 SkyBlue:LinearColor=<rgba 0 1.5 3 1>
    R1 TitaniumWhite:LinearColor=<rgba 3 3 3 1>
    R2 Black:LinearColor=<rgba 0.1 0.1 0.1 1>
    R2 BurntSienna:LinearColor=<rgba 0.55 0.11 0 1>
    R2 Cobalt:LinearColor=<rgba 0 0.4167 1 1>
    R2 Crimson:LinearColor=<rgba 1 0 0 1>
    R2 DefaultColor:LinearColor=<rgba 2 0.0667 0 1>
    R2 ForestGreen:LinearColor=<rgba 0 0.8 0 1>
    R2 Grey:LinearColor=<rgba 0.5 0.5 0.5 1>
    R2 Lime:LinearColor=<rgba 0.925 1.5 0 1>
    R2 Orange:LinearColor=<rgba 1.3 0.1733 0 1>
    R2 Pink:LinearColor=<rgba 1.2 0.36 1.06 1>
    R2 Purple:LinearColor=<rgba 0.4167 0 1 1>
    R2 Saffron:LinearColor=<rgba 1.2133 1.3 0 1>
    R2 SkyBlue:LinearColor=<rgba 0 0.65 1.3 1>
    Saffron=<rgba 0.8867 0.95 0 1>
    SkyBlue=<rgba 0 0.475 0.95 1>
    TitaniumWhite=<rgba 0.6 0.6 0.6 1>
    Use R1 Painted Overrides=true
    Use R2 Painted Overrides=true
   )
   &InitializeParticle(
    Color=$Particles.Color
    Lifetime=&MergeHelper_FloatSwitch(
     0=&RandomRangeFloat(Minimum=0.3, Maximum=0.45)
     1=&RandomRangeFloat(Minimum=0.3, Maximum=0.35)
     2=&RandomRangeFloat(Minimum=0.3, Maximum=0.4)
     3=?
     Index=&SpawnGroup()
     Num=::NewEnumerator1
    )
    Mesh Scale Mode=::NewEnumerator1
    Mesh Uniform Scale Max=0.25
    Mesh Uniform Scale Min=0.2
    Sprite Rotation Mode=::NewEnumerator2
    Sprite Size Mode=::NewEnumerator3
    Uniform Sprite Size=&MergeHelper_FloatSwitch(
     0=&RandomRangeFloat(Minimum=40, Maximum=45)
     1=1
     2=&RandomRangeFloat(Minimum=6, Maximum=12)
     3=?
     Index=&SpawnGroup()
     Num=::NewEnumerator1
    )
    UsePositionOffset=false
    Write Material Random=true
   )
   &PartitionParticles()
   &SkeletalMeshLocation(
    Bone / Socket Index=&Add_Integer(A=$Particles.PartitionParticles.Partition, B=4)
    Bone / Socket Sampling Mode=::NewEnumerator5
    Interpolate Mesh Sampling=false
    MaskSpawnGroup=false
    Mesh Sampling Type=::NewEnumerator6
    Skeletal Mesh=$System.SkeletalMesh
   )
   &ShapeLocation(
    MaskSpawnGroup=false
    Offset=&AddVectorToPosition(
     Position=&MergeHelper_VectorSwitch(0=<-25 0 5>, 1=<-20 0 10>, 2=<-20 0 5>, 3=?, Index=&SpawnGroup(), Num=::NewEnumerator1)
     Vector=$User.offset
    )
    Offset Coordinate Space=::Local
    Offset Mode=::NewEnumerator0
    Override Seed=false
    Sphere Radius=&MergeHelper_FloatSwitch(0=5, 1=25, 2=20, 3=?, Index=&SpawnGroup(), Num=::NewEnumerator1)
    UseOwnerScale=true
   )
   &AddVelocity(Velocity=&MergeHelper_VectorSwitch(
     0=&RandomRangeVector(Minimum=<-100 -5 -5>, Maximum=<-300 5 5>)
     1=&RandomRangeVector(Minimum=<-1200 0 0>, Maximum=<-1500 1 1>)
     2=&RandomRangeVector(Minimum=<-400 0 0>, Maximum=<-1400 0 0>)
     3=?
     Index=&SpawnGroup()
     Num=::NewEnumerator1
    ))
   &InitialMeshOrientation(Use Rotation Vector=false)
   /SetVariables(Particles.VisibilityTag=$Emitter.SpawnGroup) {isSetParameters=true}
  ]
  particleUpdateScript [
   &FN_DisableTireFX(
    Mask Index Attribute:NiagaraInt32=$Particles.PartitionParticles.Partition
    Mask Index FL=2
    Mask Index FR=3
    Mask Index RL=0
    Mask Index RR=1
   )
   &ParticleState()
   &ScaleMeshSize(Scale Factor=&VectorFromFloat(Value=&FloatFromCurve(FloatCurve=DataInterface() {propertyValues {Curve {keys [<at 0 0.2>,<at 0.15 1>,<at 0.7 1>,<at 1 0>]}}})))
   &ScaleSpriteSize(
    Scale Sprite Size Mode=::NewEnumerator0
    Uniform Scale Factor=&MergeHelper_FloatSwitch(
     0=&FloatFromCurve(FloatCurve=DataInterface() {propertyValues {Curve {keys [
          {arriveTangent=0.3427, interpMode=RCIM_Cubic, leaveTangent=0.3427, leaveTangentWeight=0.3333, tangentMode=RCTM_User, time=0, value=0.8}
          {arriveTangent=0.0702, arriveTangentWeight=0.3333, interpMode=RCIM_Cubic, leaveTangent=0.0702, tangentMode=RCTM_User, time=1, value=1}
         ]}}})
     1=1
     2=&FloatFromCurve(FloatCurve=DataInterface() {propertyValues {Curve {
         defaultValue=1
         keys [
          {arriveTangent=2.9549, interpMode=RCIM_Cubic, leaveTangent=2.9549, tangentMode=RCTM_User, time=0, value=0.5}
          {interpMode=RCIM_Cubic, tangentMode=RCTM_User, time=0.25, value=1}
         ]
        }}})
     3=?
     Index=&SpawnGroup()
     Num=::NewEnumerator1
    )
   )
   &Drag(
    Drag=&MergeHelper_FloatSwitch(
     0=&FloatFromCurve(FloatCurve=DataInterface() {propertyValues {Curve {
         defaultValue=0
         keys [
          {arriveTangent=1.3255, interpMode=RCIM_Cubic, leaveTangent=1.3255, tangentMode=RCTM_User, time=0, value=0}
          {interpMode=RCIM_Cubic, tangentMode=RCTM_User, time=0.72, value=1}
         ]
        }}})
     1=0
     2=0.4
     3=?
     Index=&SpawnGroup()
     Num=::NewEnumerator1
    )
    Use Linear Drag=true
    Use Rotational Drag=false
   )
   &InheritSourceMovement(
    Applied Position Scale=&Multiply_VectorByFloat(
     Vector=<1 1 1>
     Float=&FloatFromCurve(FloatCurve=DataInterface() {propertyValues {Curve {keys [<at 0 1 RCIM_Cubic>, {arriveTangent=-3.1552, interpMode=RCIM_Cubic, leaveTangent=-3.1552, tangentMode=RCTM_User, time=1, value=0}]}}})
    )
    ApplyForce=false
    ApplyPosition=true
    ApplyPositionFalloff=false
   )
   &GravityForce(Gravity=&MergeHelper_VectorSwitch(
     0=<0 0 0>
     1=<0 0 0>
     2=&VectorFromCurve(
      VectorCurve=DataInterface() {propertyValues {
        XCurve {keys [<at 0 0>,<at 1 0>]}
        YCurve {keys [<at 0 0>,<at 1 0>]}
        ZCurve {keys [{arriveTangent=-1.355, interpMode=RCIM_Cubic, leaveTangent=-1.355, tangentMode=RCTM_User, time=0, value=-0.31}, <at 1 -1 RCIM_Cubic>]}
       }}
      Scale Curve=<0 0 400>
     )
     3=?
     Index=&SpawnGroup()
     Num=::NewEnumerator1
    ))
   &SolveForcesAndVelocity(Clamp Velocity=false, Limit Acceleration=false)
   &ScaleColor(
    Scale Alpha=&FloatFromCurve(FloatCurve=DataInterface() {propertyValues {Curve {keys [
         {arriveTangent=1.0678, interpMode=RCIM_Cubic, leaveTangent=1.0678, tangentMode=RCTM_User, time=0, value=0.4}
         {arriveTangent=-0.3378, interpMode=RCIM_Cubic, leaveTangent=-0.3378, tangentMode=RCTM_User, time=0.1683, value=0.8585}
         {arriveTangent=-0.5514, interpMode=RCIM_Cubic, leaveTangent=-0.5514, time=0.3385, value=0.4958}
         {arriveTangent=-0.1597, arriveTangentWeight=0.2545, interpMode=RCIM_Cubic, leaveTangent=-0.1597, tangentMode=RCTM_User, time=1, value=0.4}
        ]}}})
    Scale RGB=&VectorFromFloat(Value=&FloatFromCurve(
      FloatCurve=DataInterface() {propertyValues {Curve {keys [
          {arriveTangent=-0.5805, interpMode=RCIM_Cubic, leaveTangent=-0.5805, tangentMode=RCTM_User, time=0, value=2}
          {arriveTangentWeight=0.3664, interpMode=RCIM_Cubic, tangentMode=RCTM_User, time=1, value=0.3}
         ]}}}
      Scale Curve=2
     ))
    ScaleA=true
    ScaleRGB=true
   )
   &SubUVAnimation(
    "Sprite Renderer "=DataInterface() {propertyValues {SpriteRenderer=&self:NiagaraEmitter_3.NiagaraSpriteRendererProperties_0}}
    UseEndFrame=false
    UseStartFrame=false
   )
  ]
  rendererClasses [&NiagaraMeshRendererProperties, &NiagaraSpriteRendererProperties]
  renderers [
   {rendererClass=&NiagaraMeshRendererProperties, rendererIndex=0}
   {rendererClass=&NiagaraSpriteRendererProperties, rendererIndex=1}
   {rendererClass=&NiagaraSpriteRendererProperties, rendererIndex=2}
  ]
 }
 NE_VC_DriftTrail_Ribbon_Parent1001 {
  emitterUpdateScript [
   &EmitterState(Inactive Response=::NewEnumerator1, Life Cycle Mode=::NewEnumerator1, Loop Duration=0.8, UseLoopDelay=false)
   &VC_RateQualityHandler(Epic Spawn Rate=50, High Spawn Rate=40, Low Spawn Rate=40, Medium Spawn Rate=40, Use Custom Rate?=true)
   &VC_SpawnRibbon_v1(Number of Ribbons=2, Spawn Time=-1, SpawnRate=$Emitter.Spawn Rate)
  ]
  particleSpawnScript [
   &VC_PaintedVariantsOverride(
    Black=<rgba 0.1 0.1 0.1 1>
    BurntSienna=<rgba 0.6 0.12 0 1>
    Cobalt=<rgba 0.16 0.3133 2 1>
    Crimson=<rgba 1.1 0 0 1>
    DefaultColor=<rgba 2 0.0667 0 1>
    ForestGreen=<rgba 0 0.8 0 1>
    Grey=<rgba 0.5 0.5 0.5 1>
    Lime=<rgba 0.925 1.5 0 1>
    Orange=<rgba 1.3 0.1733 0 1>
    Pink=<rgba 1.2 0.36 1.06 1>
    Purple=<rgba 0.4167 0 1 1>
    R1 Black:LinearColor=<rgba 0.05 0.05 0.05 1>
    R1 BurntSienna:LinearColor=<rgba 0.05 0.05 0.05 1>
    R1 Cobalt:LinearColor=<rgba 0 0.007 0.5 1>
    R1 Crimson:LinearColor=<rgba 0.05 0.05 0.05 1>
    R1 DefaultColor=<rgba 0.5 0.0575 0.05 1>
    R1 ForestGreen:LinearColor=<rgba 0.05 0.05 0.05 1>
    R1 Gold:LinearColor=<rgba 0.05 0.05 0.05 1>
    R1 Grey:LinearColor=<rgba 0.05 0.05 0.05 1>
    R1 Lime:LinearColor=<rgba 0.05 0.05 0.05 1>
    R1 Orange:LinearColor=<rgba 0.05 0.05 0.05 1>
    R1 Pink:LinearColor=<rgba 0.05 0.05 0.05 1>
    R1 Purple:LinearColor=<rgba 0.05 0.05 0.05 1>
    R1 Saffron:LinearColor=<rgba 0.05 0.05 0.05 1>
    R1 SkyBlue:LinearColor=<rgba 0.05 0.05 0.05 1>
    R1 TitaniumWhite:LinearColor=<rgba 1 1 1 1>
    Saffron=<rgba 1.2133 1.3 0 1>
    SkyBlue=<rgba 0 0.65 1.3 1>
    Use R1 Painted Overrides=true
   )
   &VC_InitializeRibbon_v1(
    Color=$Particles.Color
    Custom Offset=true
    Enable Skeletal Position=true
    Lifetime=0.55
    Local Vector Offset=&AddVectorToPosition(Position=<30 0 5>, Vector=$User.offset)
    Offset Amount=1
    Ribbon Width=20
   )
   &AddVelocity(Velocity=<-1800 0 0>) {module=AddVelocity001}
  ]
  particleUpdateScript [
   &FN_DisableTireFX(Mask Index Attribute:NiagaraInt32=$Particles.SpawnGroup, Mask Index FL=2, Mask Index FR=3, Mask Index RL=0, Mask Index RR=1)
   &ParticleState()
   &InheritSourceMovement(ApplyForce=false, ApplyPosition=true, ApplyPositionFalloff=false)
   &SolveForcesAndVelocity(Clamp Velocity=false, Limit Acceleration=false)
   &VC_UpdateRibbon_v1(Enable Skeletal Position=true)
   &ScaleRibbonWidth(Ribbon Width Scale=&FloatFromCurve(FloatCurve=DataInterface() {propertyValues {Curve {keys [<at 0 1 RCIM_Cubic>, {interpMode=RCIM_Cubic, leaveTangent=-0.2059, time=1, value=0.5}]}}}))
   /SetVariables(
    Particles.scale_R1:LinearColor=&MultiplyLinearColor(
     Linear Color=$Particles.R1 Color
     Scale Factor=&MakeVector4FromVectorAndFloat(
      "Vector (RGB)"=&VectorFromFloat(Value=&FloatFromCurve(FloatCurve=DataInterface() {propertyValues {Curve {
           defaultValue=1
           keys [{interpMode=RCIM_Cubic, tangentMode=RCTM_User, time=0.5, value=1}, {interpMode=RCIM_Cubic, tangentMode=RCTM_User, time=0.7, value=0}]
          }}}))
      "Float (Alpha)"=1
     )
    )
    Particles.width2:NiagaraFloat=&Multiply_Float(
     A=30
     B=&FloatFromCurve(FloatCurve=DataInterface() {propertyValues {Curve {keys [
          {arriveTangent=-0.611, interpMode=RCIM_Cubic, leaveTangent=-0.611, tangentMode=RCTM_User, time=0, value=1}
          {interpMode=RCIM_Cubic, leaveTangent=-0.2059, time=1, value=0.45}
         ]}}})
    )
   ) {isSetParameters=true}
   &ScaleColor(
    Scale RGB=&VectorFromFloat(Value=&FloatFromCurve(
      FloatCurve=DataInterface() {propertyValues {Curve {keys [{arriveTangent=-14.9108, interpMode=RCIM_Cubic, leaveTangent=-14.9108, tangentMode=RCTM_User, time=0.15, value=5}, <at 0.5 0.4 RCIM_Cubic>]}}}
      Scale Curve=2
     ))
    ScaleA=true
    ScaleRGB=true
   )
  ]
  rendererClasses [&NiagaraRibbonRendererProperties]
  renderers [
   {rendererClass=&NiagaraRibbonRendererProperties}
   {rendererClass=&NiagaraRibbonRendererProperties, rendererIndex=1}
  ]
 }
 NE_VC_DriftTrail_Swirls001 {
  emitterUpdateScript [
   &EmitterState(Inactive Response=::NewEnumerator1, Life Cycle Mode=::NewEnumerator1, Loop Duration=0.1, UseLoopDelay=false)
   &SpawnRate(SpawnRate=31, Use Spawn Probability=false)
   &SpawnBurst_Instantaneous(Spawn Count=2, Spawn Group=1, Use Spawn Probability=false, UseLoopCountLimit=false)
  ]
  particleSpawnScript [
   &VC_PaintedVariantsOverride(
    Black=<rgba 0.05 0.05 0.05 1>
    BurntSienna=<rgba 0.25 0.05 0 1>
    Cobalt=<rgba 0.04 0.1167 0.5 1>
    Crimson=<rgba 0.3 0 0 1>
    DefaultColor=<rgba 0.5 0.0167 0 1>
    ForestGreen=<rgba 0 0.5 0 1>
    Grey=<rgba 0.1 0.1 0.1 1>
    Lime=<rgba 0.185 0.3 0 1>
    Orange=<rgba 0.4 0.0533 0 1>
    Pink=<rgba 0.3 0.09 0.265 1>
    Purple=<rgba 0.125 0 0.3 1>
    R1 Black:LinearColor=<rgba 0.05 0.05 0.05 1>
    R1 BurntSienna:LinearColor=<rgba 0.55 0.11 0 1>
    R1 Cobalt:LinearColor=<rgba 0.16 0.3133 2 1>
    R1 Crimson:LinearColor=<rgba 1 0 0 1>
    R1 DefaultColor=<rgba 2 0.0667 0 1>
    R1 ForestGreen:LinearColor=<rgba 0 0.8 0 1>
    R1 Grey:LinearColor=<rgba 0.5 0.5 0.5 1>
    R1 Lime:LinearColor=<rgba 0.925 1.5 0 1>
    R1 Orange:LinearColor=<rgba 1.3 0.1733 0 1>
    R1 Pink:LinearColor=<rgba 1.2 0.36 1.06 1>
    R1 Purple:LinearColor=<rgba 0.4167 0 1 1>
    R1 Saffron:LinearColor=<rgba 1.2133 1.3 0 1>
    R1 SkyBlue:LinearColor=<rgba 0 0.65 1.3 1>
    Saffron=<rgba 0.28 0.3 0 1>
    SkyBlue=<rgba 0 0.15 0.3 1>
    TitaniumWhite=<rgba 0.3 0.3 0.3 1>
    Use R1 Painted Overrides=true
   )
   &InitializeParticle(
    Color=$Particles.Color
    Lifetime=&MergeHelper_FloatSwitch(
     0=&RandomRangeFloat(Minimum=0.4, Maximum=0.6)
     1=0.3
     2=?
     3=?
     Index=&SpawnGroup()
     Num=::NewEnumerator0
    )
    Mesh Scale=&Multiply_VectorByFloat(Vector=<-0.45 0.15 0.15>, Float=1.3)
    Mesh Scale Mode=::NewEnumerator4
    Sprite Rotation Mode=::NewEnumerator2
    Sprite Size=<1.5 15>
    Sprite Size Mode=::NewEnumerator4
    UsePositionOffset=false
    Write Material Random=true
   )
   &PartitionParticles()
   &SkeletalMeshLocation(
    Bone / Socket Index=&Add_Integer(A=$Particles.PartitionParticles.Partition, B=4)
    Bone / Socket Sampling Mode=::NewEnumerator5
    Interpolate Mesh Sampling=false
    MaskSpawnGroup=false
    Mesh Sampling Type=::NewEnumerator6
    Sampled Position Offset=&AddVectorToPosition(
     Position=&MergeHelper_VectorSwitch(
      0=&RandomRangeVector(Minimum=<-15 -20 -10>, Maximum=<-25 20 20>)
      1=<-25 0 5>
      2=?
      3=?
      Index=&SpawnGroup()
      Num=::NewEnumerator0
     )
     Vector=$User.offset
    )
    Skeletal Mesh=$System.SkeletalMesh
   )
   &AddVelocity(
    Velocity=&MaskVectorBySpawnGroup(
     Masked Vector=&RandomRangeVector(Minimum=<150 -15 -15>, Maximum=<450 10 10>)
     Passthrough Vector=<0 0 0>
    )
    Velocity Speed Scale=-1
   )
   /SetVariables(Particles.VisibilityTag=$Emitter.SpawnGroup) {isSetParameters=true}
   &InitialMeshOrientation(
    Mesh Axis To Orient=<-1 0 0>
    Mesh Orientation Mode=::NewEnumerator2
    Rotation=&Lerp_Vector(A=<-1 0 0>, B=<1 0 0>, Alpha=&Sine(Normalized Angle=$Emitter.Age, Period=0.1, Scale=1, Bias=0))
    Use Rotation Vector=true
   )
  ]
  particleUpdateScript [
   &FN_DisableTireFX(
    Mask Index Attribute:NiagaraInt32=$Particles.PartitionParticles.Partition
    Mask Index FL=2
    Mask Index FR=3
    Mask Index RL=0
    Mask Index RR=1
   )
   &ParticleState()
   &ScaleSpriteSize(
    Uniform Curve Scale=1.2
    Uniform Curve Sprite Scale=DataInterface() {propertyValues {Curve {keys [
        {arriveTangent=-0.0405, interpMode=RCIM_Cubic, leaveTangent=-0.0405, tangentMode=RCTM_User, time=0.3, value=1}
        {arriveTangent=0.0388, interpMode=RCIM_Cubic, leaveTangent=0.0388, tangentMode=RCTM_User, time=1, value=0}
       ]}}}
   )
   &ScaleMeshSize(Scale Factor=&Multiply_Vector(
     A=&VectorFromFloat(Value=1)
     B=&MakeVector(
      X=1
      Y=&MergeHelper_FloatSwitch(0=1, 1=-1, 2=?, 3=?, Index=$Particles.PartitionParticles.Partition, Num=::NewEnumerator0)
      Z=1
     )
    ))
   &VortexForce(
    Origin Pull Amount:NiagaraFloat=&MaskFloatBySpawnGroup(Masked Float=&RandomRangeFloat(Minimum=25000, Maximum=20000), Passthrough Float=0)
    UseOriginOffset=true
    Vortex Axis=<-1 0 0>
    Vortex Axis Coordinate Space:ENiagaraCoordinateSpace=::Local
    Vortex Force Amount=&MaskFloatBySpawnGroup(
     Masked Float=&RandomRangeFloat(Minimum=5000, Maximum=7500, Evaluation Type=::NewEnumerator1)
     Passthrough Float=0
    )
    Vortex Origin:NiagaraPosition=$Particles.Initial.Position
    Vortex Origin Offset:Vector3f=<0 1 0>
    Vortex Origin Offset Coordinate Space:ENiagaraCoordinateSpace=::Local
   )
   &SolveForcesAndVelocity(Clamp Velocity=false, Limit Acceleration=false)
   &ScaleColor(Scale Alpha=5, Scale RGB=&VectorFromFloat(Value=5), ScaleA=true, ScaleRGB=true)
   /SetVariables(Particles.sparkscolor:LinearColor=&MultiplyLinearColor(
     Linear Color=$Particles.R1 Color
     Scale Factor=&MakeVector4FromVectorAndFloat(
      "Vector (RGB)"=&VectorFromFloat(Value=5)
      "Float (Alpha)"=&FloatFromCurve(
       FloatCurve=DataInterface() {propertyValues {Curve {
          defaultValue=1
          keys [
           {arriveTangent=1.0371, interpMode=RCIM_Cubic, leaveTangent=1.0371, leaveTangentWeight=0.1667, tangentMode=RCTM_User, time=0, value=0}
           {arriveTangentWeight=0.1667, interpMode=RCIM_Cubic, leaveTangentWeight=0.1667, tangentMode=RCTM_User, time=0.5, value=1}
           {arriveTangent=-3.3261, arriveTangentWeight=0.1667, interpMode=RCIM_Cubic, leaveTangent=-3.3261, tangentMode=RCTM_User, time=1, value=0}
          ]
         }}}
       Scale Curve=2
      )
     )
    )) {isSetParameters=true}
  ]
  rendererClasses [&NiagaraMeshRendererProperties, &NiagaraSpriteRendererProperties]
  renderers [
   {rendererClass=&NiagaraMeshRendererProperties, rendererIndex=0}
   {rendererClass=&NiagaraMeshRendererProperties, rendererIndex=1}
   {rendererClass=&NiagaraSpriteRendererProperties, rendererIndex=2}
  ]
 }
 NE_VC_DriftTrail_WheelWrap_Parent1001 {
  emitterUpdateScript [
   &EmitterState(
    Inactive Response=::NewEnumerator1
    Life Cycle Mode=::NewEnumerator1
    Loop Behavior=::NewEnumerator1
    Loop Duration Mode=::NewEnumerator1
    UseLoopDelay=false
   )
   &SpawnBurst_Instantaneous(Spawn Count=2, Use Spawn Probability=false, UseLoopCountLimit=false)
   &SpawnBurst_Instantaneous(Spawn Count=2, Spawn Time=1.0667, Use Spawn Probability=false, UseLoopCountLimit=false) {module=SpawnBurst_Instantaneous001}
  ]
  particleSpawnScript [
   &VC_PaintedVariantsOverride(
    Black=<rgba 0.1 0.1 0.1 1>
    BurntSienna=<rgba 0.6 0.12 0 1>
    Cobalt=<rgba 0.16 0.3133 2 1>
    Crimson=<rgba 1.1 0 0 1>
    DefaultColor=<rgba 2 0.0667 0 1>
    ForestGreen=<rgba 0 0.8 0 1>
    Grey=<rgba 0.5 0.5 0.5 1>
    Lime=<rgba 0.925 1.5 0 1>
    Orange=<rgba 1.3 0.1733 0 1>
    Pink=<rgba 1.2 0.36 1.06 1>
    Purple=<rgba 0.4167 0 1 1>
    Saffron=<rgba 1.2133 1.3 0 1>
    SkyBlue=<rgba 0 0.65 1.3 1>
   )
   &InitializeParticle(
    Color=$Particles.Color
    Lifetime=$Emitter.CurrentLoopDuration
    Mesh Scale Mode=::NewEnumerator4
    UsePositionOffset=false
    Write Material Random=true
   )
   &PartitionParticles()
   &SkeletalMeshLocation(
    Bone / Socket Index=&Add_Integer(A=$Particles.PartitionParticles.Partition, B=4)
    Bone / Socket Sampling Mode=::NewEnumerator5
    Interpolate Mesh Sampling=false
    MaskSpawnGroup=false
    Mesh Sampling Type=::NewEnumerator6
    Sampled Position Offset=<0 1.5 0>
    Skeletal Mesh=$System.SkeletalMesh
   )
  ]
  particleUpdateScript [
   &FN_DisableTireFX(
    Mask Index Attribute:NiagaraInt32=$Particles.PartitionParticles.Partition
    Mask Index FL=2
    Mask Index FR=3
    Mask Index RL=0
    Mask Index RR=1
   )
   &VC_WheelWrapScale(Wheel Index Attribute:NiagaraInt32=&Add_Integer(A=$Particles.PartitionParticles.Partition, B=2))
   &ParticleState(Kill Particles When Lifetime Has Elapsed=false)
   &UpdateMeshOrientation(Rotation Coordinate Space=::NewEnumerator2, Rotation Vector=<0 1 0>) {enabled=false}
   &SkeletalMeshLocation(
    Bone / Socket Index=&Add_Integer(A=$Particles.PartitionParticles.Partition, B=4)
    Bone / Socket Sampling Mode=::NewEnumerator5
    MaskSpawnGroup=false
    Mesh Sampling Type=::NewEnumerator6
    Skeletal Mesh=$System.SkeletalMesh
   ) {module=SkeletalMeshLocation001}
   &SolveForcesAndVelocity(Clamp Velocity=false, Limit Acceleration=false)
   &ScaleColor(Scale Alpha=3, Scale RGB=&VectorFromFloat(Value=4), ScaleA=true, ScaleRGB=true)
  ]
  rendererClasses [&NiagaraMeshRendererProperties]
  renderers [{rendererClass=&NiagaraMeshRendererProperties, rendererIndex=0}]
 }
}
summary {
 systemName=NS_DriftTrail_Magma_MP
 emitters [
  {emitterName=NE_VC_DriftTrail_DB001, rendererClasses [&NiagaraSpriteRendererProperties]}
  {
   emitterName=NE_VC_DriftTrail_Swirls001
   rendererClasses [&NiagaraMeshRendererProperties, &NiagaraSpriteRendererProperties]
  }
  {emitterName=NE_VC_DriftTrail_Ribbon_Parent1001, rendererClasses [&NiagaraRibbonRendererProperties]}
  {
   emitterName=NE_VC_DriftTrail_Fire001
   rendererClasses [&NiagaraMeshRendererProperties, &NiagaraSpriteRendererProperties]
  }
  {emitterName=NE_VC_DriftTrail_WheelWrap_Parent1001, rendererClasses [&NiagaraMeshRendererProperties]}
 ]
 userVariables [
  User.bIsFrontendPreview {type=NiagaraBool}
  User.bIsPainted {type=NiagaraBool}
  User.ColorVariantID {type=NiagaraInt32}
  User.offset {type=Vector3f, description="offset from wheels(set through cosmetic template)", defaultValue {struct=Vector3f, value=<0 0 0>}}
  User.PaintedColor {type=LinearColor, description="", defaultValue {struct=LinearColor, value=<rgba 1 1 1 1>}}
  User.PaintedScalar {type=NiagaraFloat, defaultValue=1, description=""}
  User.VehicleFlags {type=NiagaraInt32, defaultValue=806289408}
  User.Wheel_Radius_Array {
   type=NiagaraDataInterfaceArrayFloat
   defaultValue {
    struct=NiagaraExt_VariableValue_DataInterface
    value {
     dataInterface=&self:NiagaraDataInterfaceArrayFloat_0
     dataInterfaceClass=&NiagaraDataInterfaceArrayFloat
    }
   }
  }
  User.Wheel_Width_Array {
   type=NiagaraDataInterfaceArrayFloat
   defaultValue {
    struct=NiagaraExt_VariableValue_DataInterface
    value {
     dataInterface=&self:NiagaraDataInterfaceArrayFloat_1
     dataInterfaceClass=&NiagaraDataInterfaceArrayFloat
    }
   }
  }
  User.WheelSize {type=NiagaraFloat, defaultValue=0, description=""}
 ]
}
